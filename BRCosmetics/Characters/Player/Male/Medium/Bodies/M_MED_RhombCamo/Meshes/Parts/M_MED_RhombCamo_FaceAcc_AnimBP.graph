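# Animation Blueprint for the RhombCamo male face accessory that copies its pose from the attached parent mesh and drives a head-relative projection plane in its material.
# Each update reads the owner mesh's head bone world transform and sets the part's dynamic material vector parameters for plane normal and X axis from negated right and up vectors.
# The plane center is the head location plus VisorPatternOffset along the bone's axes, apparently anchoring a visor pattern to the head.
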AnimBlueprint /BRCosmetics/Characters/Player/Male/Medium/Bodies/M_MED_RhombCamo/Meshes/Parts/M_MED_RhombCamo_FaceAcc_AnimBP
target_skeleton=/Game/Characters/Player/Male/Medium/Bodies/M_MED_RhombCamo/Meshes/Parts/M_MED_RhombCamo_FaceAcc_Skeleton
graphs [
 AnimGraph {nodes [Output Pose(Result=Copy Pose From Mesh() {bUseAttachedParent=true}) {layerGroup=None}]}
 EventGraph {nodes [
   Event BlueprintUpdateAnimation()
   > n1 = Get Dynamic Material Instance(SkeletalMeshComponent=GetPartSkeletalMeshComponent())
   n2 = BreakTransform(InTransform=GetBoneTransform!(
     self:Skinned Mesh Component Object Reference=GetOwnerSkeletalMeshComponent()
     InBoneName:Name=head
     TransformSpace:ERelativeTransformSpace Enum=RTS_World
    ) -> (ReturnValue:Transform))
   n3 = GetRightVector!(InRot:Rotator=@n2.Rotation) -> (ReturnValue:Vector)
   > SetVectorParameterValue(
    self=@n1
    ParameterName=VP_PlaneNormal
    Value="To LinearColor (Vector)"!(InVec:Vector=vector * vector(A=@n3, B:Vector="-1,-1,-1") -> (ReturnValue=~)) -> (ReturnValue:Linear Color Structure)
   )
   n4 = GetUpVector!(InRot:Rotator=@n2.Rotation) -> (ReturnValue:Vector)
   > SetVectorParameterValue(self=@n1, ParameterName=VP_XAxis, Value="To LinearColor (Vector)"(InVec=vector * vector(A=@n4, B="-1,-1,-1") -> (ReturnValue=~)))
   n5 = BreakVector(InVec=Get VisorPatternOffset() -> (VisorPatternOffset:Vector))
   n6 = vector * vector(A=@n3, B:double=@n5) -> (ReturnValue=~)
   n7 = vector * vector(A=@n4, B=@n5.Y) -> (ReturnValue=~)
   n8 = vector * vector(A=GetForwardVector!(InRot:Rotator=@n2.Rotation) -> (ReturnValue:Vector), B=@n5.Z) -> (ReturnValue=~)
   > SetVectorParameterValue(
    self=@n1
    ParameterName=VP_PlaneCenter
    Value="To LinearColor (Vector)"(InVec=vector + vector!(A=@n2, B=@n6, C:Vector=@n7, D:Vector=@n8) -> (ReturnValue))
   )
  ]}
]
variables [VisorPatternOffset {type=Vector, default=<9.7 1 5>}]
